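; Parent waypoint ring for Papaya minigame trials, serving as time trial start, checkpoint or finish, or as a score ring that briefly hides.
; Once playlist data is ready, it seeds the scoreboard or, if disabled, destroys itself.
; Overlaps check vehicle tags and send gameplay events to a time trial ability, granted through a separate collision.
; Construction configures mesh, materials and linked-waypoint spline; ring animations use timestamps in custom primitive data.
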(event EventBeginPlay
  (bind _externalscoreboardactor (Variables|Default|GetExternalScoreboardActor))
  ;[1 1 1] BeginPlay -> Enabled Checks
  (Class|FortAsyncActionCurrentPlaylistReady|PlaylistDataReadyAsync
    (:OnReady
      (bind _enabled (CallFunction|IsTrialEnabled))
      (if _enabled
        ;[1 1 1] Setup the collision shape to use for overlaps
        (CallFunction|SetCollisionShape)
        (Utilities|Time|SetTimerbyEvent (EventDispatchers|CreateEvent :!event InitScoreboard) 1e+01)
        (else
          (Utilities|IsValid _externalscoreboardactor
            (:Is%20Valid
              (Actor|DestroyActor _externalscoreboardactor))
            (:Is%20Not%20Valid
              _))
          (Actor|DestroyActor))))))

(event OnComponentBeginOverlap(GrantAbilityCollision)
  (params (OverlappedComponent :type Primitive%20Component%20Object%20Reference) (OtherActor :type Actor%20Object%20Reference) (OtherComp :type Primitive%20Component%20Object%20Reference) (OtherBodyIndex :type Integer) (bFromSweep :type Boolean) (SweepResult :type Hit%20Result%20Structure%20%28by%20ref%29))
  ;[1 1 1] Grant the Time Trial Ability. This should be always granted but can't find a way to do that through an AbilitySet in a plugin right now
  (bind _asfort_player_pawn (Utilities|Casting|CastToFortPlayerPawn OtherActor))
  (Utilities|IsValid (Ability|GetFortAbilitySystemComponent _asfort_player_pawn)
    (:Is%20Valid
      (GameplayEffects|ApplyGameplayEffectToSelf :self (Ability|GetFortAbilitySystemComponent _asfort_player_pawn) :GameplayEffectClass (Variables|Default|GetGE_TimeTrialAbility) :Level 0.0))))

(event Custom|AnimateRingOut
  ;[1 1 1] Animate rings in and out
  (Rendering|Material|SetCustomPrimitiveDataFloat (Variables|Default|GetWayPointMesh) 0 (Utilities|Time|GetGameTimeinSeconds))
  (Rendering|Material|SetCustomPrimitiveDataFloat (Variables|Default|GetWayPointMesh) 1 (Variables|Default|GetAnimateOutLength))
  (Rendering|Material|SetCustomPrimitiveDataFloat (Variables|Default|GetWayPointMesh) 3 1.0)
  (bind _returnvalue (Utilities|Time|SetTimerbyEvent (delegate Custom|HideActor) (Variables|Default|GetAnimateOutLength)))
  (Variables|Default|SetAnimateRingOutTimer _returnvalue))

(event OnComponentBeginOverlap(SphereCollision)
  (params (OverlappedComponent :type Primitive%20Component%20Object%20Reference) (OtherActor :type Actor%20Object%20Reference) (OtherComp :type Primitive%20Component%20Object%20Reference) (OtherBodyIndex :type Integer) (bFromSweep :type Boolean) (SweepResult :type Hit%20Result%20Structure%20%28by%20ref%29))
  ;[1 1 1] Try to start a Trial on Overlap (only one collision shape should be active at a time)
  (CallFunction|OnOverlapStart OtherActor))

(event Custom|HideActor
  ;[1 1 1] Animate rings in and out
  (CallFunction|SetActorHidden true))

(event OnComponentBeginOverlap(CylinderCollision)
  (params (OverlappedComponent :type Primitive%20Component%20Object%20Reference) (OtherActor :type Actor%20Object%20Reference) (OtherComp :type Primitive%20Component%20Object%20Reference) (OtherBodyIndex :type Integer) (bFromSweep :type Boolean) (SweepResult :type Hit%20Result%20Structure%20%28by%20ref%29))
  ;[1 1 1] Try to start a Trial on Overlap (only one collision shape should be active at a time)
  (CallFunction|OnOverlapStart OtherActor))

(event Custom|AnimateRingIn
  ;[1 1 1] Animate rings in and out
  (Utilities|Time|ClearandInvalidateTimerbyHandle (Variables|Default|GetAnimateRingOutTimer))
  (CallFunction|SetActorHidden false)
  (Rendering|Material|SetCustomPrimitiveDataFloat (Variables|Default|GetWayPointMesh) 0 (Utilities|Time|GetGameTimeinSeconds))
  (Rendering|Material|SetCustomPrimitiveDataFloat (Variables|Default|GetWayPointMesh) 1 (Variables|Default|GetAnimateInLength))
  (Rendering|Material|SetCustomPrimitiveDataFloat (Variables|Default|GetWayPointMesh) 3 0.0))

(event Custom|ScoreWaypointOverlapped
  (params (OverlappedActor :type Actor%20Object%20Reference))
  ;[1 1 1] Score.  Currently just dis and reappers.  Not tracking score anywhere the way we are time.
  (if (Pawn|IsLocallyControlled (Variables|Actor|GetInstigator OverlappedActor))
    (Rendering|SetActorHiddenInGame true)
    (Audio|PlaySound2D (Variables|Default|GetScoreUpSound))
    (Utilities|Time|SetTimerbyEvent (delegate Custom|ToggleScoreRingVis) 4.0)))

(event Custom|ToggleScoreRingVis
  ;[1 1 1] Score.  Currently just dis and reappers.  Not tracking score anywhere the way we are time.
  (Rendering|SetActorHiddenInGame false))

(event Custom|TimeTrialWaypointOverlapped
  (params (OverlappedActor :type Actor%20Object%20Reference))
  ;[1 1 1] TimeTirals
  (if (Variables|Default|GetStartsTrial)
    (if (not (GameplayTags|HasMatchingGameplayTag (Ability|GetFortAbilitySystemComponent OverlappedActor) (Variables|Default|GetT_GrantedTimeTrial)))
      ;;[1 1 1] Start trial event
      (Ability|SendGameplayEventtoActor OverlappedActor (Variables|Default|GetT_EventTimeTrialStart) (Utilities|Struct|MakeGameplayEventData :OptionalObject self :ContextHandle "()" :InstigatorTags "(GameplayTags=)" :TargetTags "(GameplayTags=)" :EventMagnitude (Variables|TimeTrial|GetBaseTrialTime) :TargetData "()")))
    (elif (Variables|Default|GetEndsTrial)
      ;;[1 1 1] Success event
      (Ability|SendGameplayEventtoActor OverlappedActor (Variables|Default|GetT_TimeTrialSuccess) (Utilities|Struct|MakeGameplayEventData :OptionalObject self :ContextHandle "()" :InstigatorTags "(GameplayTags=)" :TargetTags "(GameplayTags=)" :EventMagnitude 0.0 :TargetData "()"))
      (else
        ;;[1 1 1] Checkpoint event
        (Ability|SendGameplayEventtoActor OverlappedActor (Variables|Default|GetT_TimeTrialCheckpoint) (Utilities|Struct|MakeGameplayEventData :OptionalObject self :ContextHandle "()" :InstigatorTags "(GameplayTags=)" :TargetTags "(GameplayTags=)" :EventMagnitude (Variables|TimeTrial|GetCheckPointTimeReward) :TargetData "()"))))))

(fn GetCenterMID1
  (bind _asmaterial_instance_dynamic (Utilities|Casting|CastToMaterialInstanceDynamic (Rendering|Material|GetMaterial (Variables|Default|GetWayPointMesh) 1))
    (:then
      (return _asmaterial_instance_dynamic))
    (:CastFailed
      (bind _returnvalue (Rendering|Material|CreateDynamicMaterialInstance (Variables|Default|GetWayPointMesh) 1 0 "None" :!class /Script/Engine.PrimitiveComponent))
      (return _returnvalue))))

(fn GetRingMID0
  (bind _asmaterial_instance_dynamic (Utilities|Casting|CastToMaterialInstanceDynamic (Rendering|Material|GetMaterial (Variables|Default|GetWayPointMesh) 0))
    (:then
      (return _asmaterial_instance_dynamic))
    (:CastFailed
      (bind _returnvalue (Rendering|Material|CreateDynamicMaterialInstance (Variables|Default|GetWayPointMesh) 0 0 "None" :!class /Script/Engine.PrimitiveComponent))
      (return _returnvalue))))

(fn InitScoreboard
  (bind _externalscoreboardactor (Variables|Default|GetExternalScoreboardActor))
  ;[1 1 1] Initial Scoreboard Time -> Default Top Score to 10 secs slower than the C Rank par time with a blank name
  (Networking|SwitchHasAuthority
    (:Authority
      (Utilities|IsValid _externalscoreboardactor
        (:Is%20Valid
          (Class|BPapayaTrialsTopScore|UpdateTopScore _externalscoreboardactor (Utilities|Struct|MakeStructPapayaTrialsTopScore (+ (Variables|TimeTrial|GetParTime_Rank3) 1e+01) "INVTEXT(\" \")" true)))))))

(fn IsTrialEnabled
  ;[1 1 1] Time Trials Enabled?
  (if (Class|AthenaFunctionLibrary|RowtoBool (Variables|Default|GetEnabledRow))
    ;[1 1 1] Requires a Vehicle?
    (if (GameplayTags|Equal(GameplayTagContainer) (Variables|TimeTrial|GetRequiredVehicleTag) (GameplayTags|MakeGameplayTagContainerfromTag))
      (return true)
      (else
        ;[1 1 1] What Vehicle is required?
        (switch Utilities|FlowControl|Switch|SwitchonGameplayTagContainer (Variables|TimeTrial|GetRequiredVehicleTag)
          (:Default
            (return true))
          (:(GameplayTags=((TagName="Vehicle.Antelope")))
            ;[1 1 1] Is Quad Enabled?
            (if (not (Class|AthenaFunctionLibrary|RowtoBool (LocalVariables|GetLOCAL_QuadEnabled)))
              (return false)))
          (:(GameplayTags=((TagName="Vehicle.Meatball")))
            ;[1 1 1] Is Boat Enabled?
            (if (not (Class|AthenaFunctionLibrary|RowtoBool (LocalVariables|GetLOCAL_BoatEnabled)))
              (return false))))
        (return true)))
    (else
      (return false))))

(fn OnOverlapStart
  (params (OverlapActor :type Actor%20Object%20Reference))
  (bind _asfort_player_pawn (Utilities|Casting|CastToFortPlayerPawn OverlapActor))
  (bind _canactivate (CallFunction|TrialTagRequirements _asfort_player_pawn))
  (if _canactivate
    (switch Utilities|FlowControl|Switch|SwitchonEnum_Papaya_Waypoint_Types (Variables|Default|GetWayPointType)
      (:NewEnumerator1
        (CallFunction|ScoreWaypointOverlapped _asfort_player_pawn))
      (:NewEnumerator2
        _)
      (:NewEnumerator5
        _))
    (CallFunction|TimeTrialWaypointOverlapped _asfort_player_pawn)))

(fn SetActorHidden
  (params (bNewHidden :type Boolean))
  (Rendering|SetActorHiddenInGame bNewHidden))

(fn SetCollisionShape
  (bind _spherecollision (Variables|Default|GetSphereCollision))
  (bind _cylindercollision (Variables|Default|GetCylinderCollision))
  (switch Utilities|FlowControl|Switch|SwitchonEnum_Papaya_Waypoint_Types (Variables|Default|GetWayPointType)
    (:NewEnumerator0
      _)
    (:NewEnumerator1
      _)
    (:NewEnumerator2
      _)
    (:NewEnumerator3
      _)
    (:NewEnumerator4
      _)
    (:NewEnumerator5
      (Components|DestroyComponent _spherecollision)
      (Collision|SetCollisionEnabled _cylindercollision "QueryOnly")))
  (Collision|SetCollisionEnabled _spherecollision "QueryOnly")
  (Components|DestroyComponent _cylindercollision))

(fn SetEndTrialMaterialParameters
  (bind _output_get (LocalVariables|SetLocalCenterMID (CallFunction|GetCenterMID1)))
  (Rendering|Material|SetVectorParameterValue _output_get "IconChannel" "(R=0.000000,G=0.000000,B=0.000000,A=1.000000)" :!class /Script/Engine.MaterialInstanceDynamic)
  (Rendering|Material|SetScalarParameterValue (LocalVariables|GetLocalCenterMID) "IconSelection" 1.0 :!class /Script/Engine.MaterialInstanceDynamic)
  (Rendering|Material|SetScalarParameterValue (LocalVariables|GetLocalCenterMID) "IconBiasU" (Utilities|Select :Index (Utilities|Map|Find (Variables|Default|GetWaypointTypeMap) (Variables|Default|GetWayPointType)) :Option%200 0.0 :Option%201 0.0 :Option%202 0.0 :Option%203 -0.375) :!class /Script/Engine.MaterialInstanceDynamic))

(fn SetInitialMaterialParameters
  (bind _waypointtype (Variables|Default|GetWayPointType))
  (bind _waypointtypemap (Variables|Default|GetWaypointTypeMap))
  (bind _value (Utilities|Map|Find _waypointtypemap _waypointtype))
  (bind _waypointtype_2 (Variables|Default|GetWayPointType))
  (bind _waypointtypemap_2 (Variables|Default|GetWaypointTypeMap))
  (bind _value_2 (Utilities|Map|Find _waypointtypemap_2 _waypointtype_2))
  (bind _waypointmesh (Variables|Default|GetWayPointMesh))
  (Utilities|FlowControl|Sequence
    (:then_0
      (Rendering|Material|CreateDynamicMaterialInstance (Variables|Default|GetWayPointMesh) 0 (Variables|Default|GetWaypointRingMaterial) "None" :!class /Script/Engine.PrimitiveComponent)
      (Rendering|Material|SetScalarParameterValue (CallFunction|GetRingMID0) "FaceTowardsCamera" (Utilities|Select :Index _value :Option%200 1.0 :Option%201 1.0 :Option%202 1.0 :Option%203 0.0) :!class /Script/Engine.MaterialInstanceDynamic)
      (Transformation|SetRelativeRotation (Variables|Default|GetWayPointMesh) (Utilities|Select :Index _value :Option%200 "0.000000,90.000000,0.000000" :Option%201 "0.000000,90.000000,0.000000" :Option%202 "0.000000,90.000000,0.000000" :Option%203 "0, 0, 0") false false))
    (:then_1
      (if (not (Variables|Default|GetCenterMaterialOverridden))
        (Rendering|Material|CreateDynamicMaterialInstance (Variables|Default|GetWayPointMesh) 1 (Variables|Default|GetWaypointCenterMaterial) "None" :!class /Script/Engine.PrimitiveComponent)
        (bind _output_get (LocalVariables|SetLocalCenterMID (CallFunction|GetCenterMID1)))
        (Rendering|Material|SetVectorParameterValue _output_get "IconChannel" "(R=0.000000,G=0.000000,B=0.000000,A=0.000000)" :!class /Script/Engine.MaterialInstanceDynamic)
        (Rendering|Material|SetScalarParameterValue (LocalVariables|GetLocalCenterMID) "NumberOfIcons" 4.0 :!class /Script/Engine.MaterialInstanceDynamic)
        (Rendering|Material|SetScalarParameterValue (LocalVariables|GetLocalCenterMID) "IconSelection" 2.0 :!class /Script/Engine.MaterialInstanceDynamic)
        (Rendering|Material|SetScalarParameterValue (LocalVariables|GetLocalCenterMID) "IconBiasU" (Utilities|Select :Index _value_2 :Option%200 0.0 :Option%201 0.0 :Option%202 0.0 :Option%203 -0.620635) :!class /Script/Engine.MaterialInstanceDynamic)
        (Rendering|Material|SetScalarParameterValue (LocalVariables|GetLocalCenterMID) "IconBiasV" (Utilities|Select :Index _value_2 :Option%200 0.0 :Option%201 0.0 :Option%202 0.0 :Option%203 -0.152381) :!class /Script/Engine.MaterialInstanceDynamic)
        (Rendering|Material|SetScalarParameterValue (LocalVariables|GetLocalCenterMID) "IconScaleUV" (Utilities|Select :Index _value_2 :Option%200 1.0 :Option%201 1.0 :Option%202 1.0 :Option%203 2.0) :!class /Script/Engine.MaterialInstanceDynamic)
        (Rendering|Material|SetScalarParameterValue (LocalVariables|GetLocalCenterMID) "FaceTowardsCamera" (Utilities|Select :Index _value_2 :Option%200 1.0 :Option%201 1.0 :Option%202 1.0 :Option%203 0.0) :!class /Script/Engine.MaterialInstanceDynamic)
        ;[1 1 1] Animations are handled through a timestamp in the material
        (Rendering|Material|SetCustomPrimitiveDataFloat _waypointmesh 0 0.0)
        (Rendering|Material|SetCustomPrimitiveDataFloat _waypointmesh 1 1.0)
        (Rendering|Material|SetCustomPrimitiveDataFloat (Variables|Default|GetWayPointMesh) 3 0.0)))))

(fn SetOverrideMaterials
  (bind _centermaterialoverride (Variables|Default|GetCenterMaterialOverride)
    (:then
      (Rendering|Material|SetMaterial (Variables|Default|GetWayPointMesh) 1 _centermaterialoverride)
      (Variables|Default|SetCenterMaterialOverridden true))
    (:else)))

(fn SetRingScale
  (Transformation|SetRelativeScale3D (Variables|Default|GetWayPointMesh) (Utilities|Select :Index (Utilities|Map|Find (Variables|Default|GetWaypointTypeMap) (Variables|Default|GetWayPointType)) :Option%200 (Variables|Default|GetRingScaleTime) :Option%201 (Variables|Default|GetRingScaleTime) :Option%202 (Variables|Default|GetRingScaleTime) :Option%203 (Variables|Default|GetRingScaleWaypoint))))

(fn SetStartTrialMaterialParameters
  (Rendering|Material|SetVectorParameterValue (CallFunction|GetCenterMID1) "IconChannel" "(R=0.000000,G=0.000000,B=0.000000,A=1.000000)" :!class /Script/Engine.MaterialInstanceDynamic))

(fn ToggleLinkedVisibility
  (params (LinkedWaypointsVisible :type Boolean))
  (for _array_element (Variables|Default|GetLinkedWaypoints)
    (if LinkedWaypointsVisible
      (CallFunction|AnimateRingIn _array_element)
      (else
        (CallFunction|AnimateRingOut _array_element))))
  (if LinkedWaypointsVisible
    (CallFunction|AnimateRingOut)
    (else
      (CallFunction|AnimateRingIn))))

(fn TrialTagRequirements
  (params (FortPlayerPawn :type Fort%20Player%20Pawn%20Object%20Reference))
  (if (GameplayTags|Equal(GameplayTagContainer) (Variables|TimeTrial|GetRequiredVehicleTag) (GameplayTags|MakeLiteralGameplayTagContainer))
    _
    (elif (Vehicle|IsinVehicle FortPlayerPawn)
      (if (not (GameplayTags|Equal(GameplayTagContainer) (Class|FortVehicleInterface|GetVehicleTags (Vehicle|GetVehicleInterface FortPlayerPawn)) (Variables|TimeTrial|GetRequiredVehicleTag)))
        (return false))
      (else
        (return false))))
  (return true))

(fn UpdateMaterialsbySequence
  (params (SequenceIndex :type Integer))
  (bind _ringcolorpalette (Variables|Default|GetRingColorPalette))
  (bind _centercolorspurple (Variables|RingColors|GetCenterColorsPurple))
  (bind _centercolorsblue (Variables|RingColors|GetCenterColorsBlue))
  (bind _centercolorsgreen (Variables|RingColors|GetCenterColorsGreen))
  (bind _centercolorsyellow (Variables|RingColors|GetCenterColorsYellow))
  (bind _centercolorsorange (Variables|RingColors|GetCenterColorsOrange))
  (bind _centercolorsred (Variables|RingColors|GetCenterColorsRed))
  (bind _returnvalue (Utilities|Select :Index _ringcolorpalette :NewEnumerator0 _centercolorsred :NewEnumerator1 _centercolorsorange :NewEnumerator2 _centercolorsyellow :NewEnumerator3 _centercolorsgreen :NewEnumerator4 _centercolorsblue :NewEnumerator5 _centercolorspurple))
  (bind _ringcolorpalette_2 (Variables|Default|GetRingColorPalette))
  (bind _ringcolorspurple (Variables|RingColors|GetRingColorsPurple))
  (bind _ringcolorsblue (Variables|RingColors|GetRingColorsBlue))
  (bind _ringcolorsgreen (Variables|RingColors|GetRingColorsGreen))
  (bind _ringcolorsyellow (Variables|RingColors|GetRingColorsYellow))
  (bind _ringcolorsorange (Variables|RingColors|GetRingColorsOrange))
  (bind _ringcolorsred (Variables|RingColors|GetRingColorsRed))
  (bind _returnvalue_2 (Utilities|Select :Index _ringcolorpalette_2 :NewEnumerator0 _ringcolorsred :NewEnumerator1 _ringcolorsorange :NewEnumerator2 _ringcolorsyellow :NewEnumerator3 _ringcolorsgreen :NewEnumerator4 _ringcolorsblue :NewEnumerator5 _ringcolorspurple))
  (Networking|CanExecuteCosmeticEvents
    (:True
      (bind _keys (Utilities|Map|Keys _returnvalue_2))
      (LocalVariables|SetKeysLocal _keys)
      (bind _values (Utilities|Map|Values _returnvalue_2))
      (bind _output_get (LocalVariables|SetValuesLocal _values))
      (LocalVariables|SetColorIndex (Math|Integer|Clamp(Integer) (Variables|GetSequenceIndex) 0 (Utilities|Array|LastIndex _output_get)))
      ;[1 1 1] Ring
      (Rendering|Material|SetVectorParameterValue (CallFunction|GetRingMID0) "ColorA" (Utilities|Array|Get(acopy) (LocalVariables|GetKeysLocal) (LocalVariables|GetColorIndex)) :!class /Script/Engine.MaterialInstanceDynamic)
      (Rendering|Material|SetVectorParameterValue (CallFunction|GetRingMID0) "ColorB" (Utilities|Array|Get(acopy) (LocalVariables|GetValuesLocal) (LocalVariables|GetColorIndex)) :!class /Script/Engine.MaterialInstanceDynamic)
      (bind _keys_2 (Utilities|Map|Keys _returnvalue))
      (LocalVariables|SetKeysLocal _keys_2)
      (bind _values_2 (Utilities|Map|Values _returnvalue))
      (bind _output_get_2 (LocalVariables|SetValuesLocal _values_2))
      (LocalVariables|SetColorIndex (Math|Integer|Clamp(Integer) (Variables|GetSequenceIndex) 0 (Utilities|Array|LastIndex _output_get_2)))
      ;[1 1 1] Center
      (Rendering|Material|SetScalarParameterValue (CallFunction|GetCenterMID1) "OpacityEdge" (Utilities|Select :Index (LocalVariables|GetColorIndex) :Option%200 1.1 :Option%201 1.05 :Option%202 0.0 :Option%203 0.0) :!class /Script/Engine.MaterialInstanceDynamic)
      (Rendering|Material|SetVectorParameterValue (CallFunction|GetCenterMID1) "ColorA" (Utilities|Array|Get(acopy) (LocalVariables|GetKeysLocal) (LocalVariables|GetColorIndex)) :!class /Script/Engine.MaterialInstanceDynamic)
      (Rendering|Material|SetVectorParameterValue (CallFunction|GetCenterMID1) "ColorB" (Utilities|Array|Get(acopy) (LocalVariables|GetValuesLocal) (LocalVariables|GetColorIndex)) :!class /Script/Engine.MaterialInstanceDynamic)
      ;[1 1 1] Next ring bounces more
      (Rendering|Material|SetScalarParameterValue (CallFunction|GetCenterMID1) "BounceAmount" (Utilities|Select :Index (LocalVariables|GetColorIndex) :Option%200 4e+01 :Option%201 6.0 :Option%202 6.0 :Option%203 6.0) :!class /Script/Engine.MaterialInstanceDynamic))))

(fn ConstructionScript
  (bind _waypointtype (Variables|Default|GetWayPointType))
  (bind _waypointtypemap (Variables|Default|GetWaypointTypeMap))
  (bind _value (Utilities|Map|Find _waypointtypemap _waypointtype))
  (bind _timetrialwaypointmesh (Variables|Default|GetTimeTrialWaypointMesh))
  (bind _scoreoverlapmesh (Variables|Default|GetScoreOverlapMesh))
  (bind _genericoverlapmesh (Variables|Default|GetGenericOverlapMesh))
  (bind _returnvalue (Utilities|Select :Index _value :Option%200 _timetrialwaypointmesh :Option%201 _genericoverlapmesh :Option%202 _scoreoverlapmesh :Option%203 _timetrialwaypointmesh))
  (bind _spherecollision (Variables|Default|GetSphereCollision))
  (bind _linkedwaypoints (Variables|Default|GetLinkedWaypoints))
  (bind _spline (Variables|Default|GetSpline))
  (bind _cylindercollision (Variables|Default|GetCylinderCollision))
  (bind _waypointmesh (Variables|Default|GetWayPointMesh))
  (bind _spherecollision_2 (Variables|Default|GetSphereCollision))
  (|CallParentFunction)
  (Components|StaticMesh|SetStaticMesh (Variables|Default|GetWayPointMesh) _returnvalue)
  (Utilities|FlowControl|Sequence
    (:then_0
      (CallFunction|SetRingScale)
      (CallFunction|SetOverrideMaterials)
      (CallFunction|SetInitialMaterialParameters)
      (CallFunction|UpdateMaterialsBySequence 0))
    (:then_1
      (if (Variables|Default|GetStartsTrial)
        (CallFunction|SetStartTrialMaterialParameters)
        ;[1 1 1] Spline
        (Utilities|IsValid (Variables|Default|GetSpline)
          (:Is%20Valid
            (if (Utilities|IsArrayNonEmpty _linkedwaypoints)
              (Utilities|Array|AddUnique (Variables|EditorOnly|GetEditorSplines) (Transformation|GetActorLocation))
              (for _array_element _linkedwaypoints
                (Utilities|IsValid _array_element
                  (:Is%20Valid
                    (Utilities|Array|AddUnique (Variables|EditorOnly|GetEditorSplines) (Transformation|GetActorLocation _array_element)))))
              (Spline|SetSplinePoints (Variables|Default|GetSpline) (Variables|EditorOnly|GetEditorSplines) "World" true))))
        (else
          (Utilities|IsValid _spline
            (:Is%20Valid
              (Components|DestroyComponent _spline) ; Only need the spline on the starting waypoint
            )))))
    (:then_2
      (if (Variables|Default|GetEndsTrial)
        (CallFunction|SetEndTrialMaterialParameters)))
    (:then_3
      (if (Utilities|Equal(Object) _returnvalue _scoreoverlapmesh)
        (Rendering|Material|SetMaterial _waypointmesh 0 (Variables|Default|GetSkyDiveRingMaterial))
        (Transformation|SetWorldScale3D _waypointmesh (Variables|Default|GetSkyDiveRingScale))))
    (:then_4
      (switch Utilities|FlowControl|Switch|SwitchonEnum_Papaya_Waypoint_Types (Variables|Default|GetWayPointType)
        (:NewEnumerator0
          _)
        (:NewEnumerator1
          _)
        (:NewEnumerator2
          _)
        (:NewEnumerator3
          _)
        (:NewEnumerator4
          _)
        (:NewEnumerator5
          (Collision|SetCollisionEnabled _spherecollision "NoCollision")
          (Collision|SetCollisionEnabled _cylindercollision "QueryOnly")
          (Rendering|SetVisibility _spherecollision_2 false false)))
      (Collision|SetCollisionEnabled _spherecollision "QueryOnly")
      (Collision|SetCollisionEnabled _cylindercollision "NoCollision")
      (Rendering|SetVisibility _spherecollision_2 true false))))
