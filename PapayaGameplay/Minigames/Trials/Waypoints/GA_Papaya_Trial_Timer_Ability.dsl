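; Gameplay ability that runs a waypoint time trial, started by an event carrying the start waypoint and time limit.
; The local client counts down on a looping timer, while gameplay events add checkpoint time or finish the run.
; Timeout, or landing early in skydive trials, fails it.
; Success scores elapsed time plus skipped-waypoint penalties, ranks it against par times, updates scoreboard and minigame UI, then net-syncs with the waiting server to end.
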(event Custom|OnPawnLanded_Event_0
  (params (Hit :type Hit%20Result%20Structure%20%28by%20ref%29))
  ;[1 1 1] Fail the skydive challenge if player lands before hitting the landing waypoint
  (if (Pawn|IsLocallyControlled (stale Ability|EventActivateAbilityFromEvent _asfort_player_pawn))
    (CallFunction|TrialFailed)))

(event Ability|EventActivateAbilityFromEvent
  (params (EventData :type Gameplay%20Event%20Data%20Structure%20%28by%20ref%29))
  (bind (_eventtag _instigator _target _optionalobject _optionalobject2 _contexthandle _instigatortags _targettags _eventmagnitude) (Utilities|Struct|BreakGameplayEventData EventData))
  (bind _waypointactor (Variables|Default|GetWaypointActor))
  (bind _returnvalue (Transformation|GetActorLocation _waypointactor))
  (bind _waypointactor_2 (Variables|Default|GetWaypointActor))
  (bind _returnvalue_2 (Transformation|GetActorLocation _waypointactor_2))
  (bind _startingwaypoint (Variables|Default|GetStartingWaypoint))
  (bind _returnvalue_3 (Transformation|GetActorLocation _startingwaypoint))
  (Development|PrintString "Activated Time Trial Ability")
  ;[1 1 1] Start Event - Setup the Timer
  (Ability|HasAuthority
    (:True
      _)
    (:False
      (Variables|Default|SetStartTime (Fortnite|Sync|Time|GetServerWorldTimeSeconds))))
  (Variables|Default|SetTrialTimeRemaining _eventmagnitude)
  (bind _asbp_waypoint_parent_papaya (Utilities|Casting|CastToBP_Waypoint_Parent_Papaya _optionalobject))
  (bind _output_get (Variables|Default|SetStartingWaypoint _asbp_waypoint_parent_papaya))
  (Variables|Default|SetLinkedWaypoints (Variables|/PapayaGameplay/Minigames/Trials/Waypoints/BP_Waypoint_Parent_Papaya.BP_Waypoint_Parent_Papaya_C|GetLinkedWaypoints _output_get))
  (Utilities|FlowControl|Sequence
    (:then_0
      ;[1 1 1] Clear Landed Event
      (bind _asfort_player_pawn (Utilities|Casting|CastToFortPlayerPawn (Variables|/Script/Engine.Actor|GetInstigator (Ability|GetAvatarActorfromActorInfo))))
      (EventDispatchers|UnbindEventfromOnPawnLanded _asfort_player_pawn (delegate Custom|OnPawnLanded_Event_0))
      ;[1 1 1] Bind Landed Event for Skydiving
      (Utilities|Casting|CastToBP_Waypoint_Papaya_Skydive_Start (Variables|Default|GetStartingWaypoint))
      (EventDispatchers|AssignOnPawnLanded _asfort_player_pawn (delegate Custom|OnPawnLanded_Event_0)))
    (:then_1
      (Utilities|Array|Clear (Variables|Default|GetClearedWaypoints))
      (Variables|Default|SetCurrentWaypointIdx 0)
      (Variables|Default|SetInstigatorPawn (Variables|/Script/Engine.Actor|GetInstigator (Ability|GetAvatarActorfromActorInfo)))
      (Ability|HasAuthority
        (:True
          ;[1 1 1] Update the minigame ui and top score board on success
          (Ability|Tasks|WaitNetSync "OnlyServerWait"
            (:OnSync
              (Ability|EndAbility))))
        (:False
          (Ability|IsLocallyControlled
            (:True
              ;[1 1 1] Start the Trial on the client
              (if (Pawn|IsLocallyControlled (Variables|Default|GetInstigatorPawn))
                (Utilities|FlowControl|Sequence
                  (:then_0
                    (bind _returnvalue_4 (Class|FortControllerComponentMinigameActivity|GetMinigameActivityComponent (Variables|Default|GetInstigatorPawn))
                      (:Valid
                        (Class|FortControllerComponentMinigameActivity|ForwardMinigameActivityStarted _returnvalue_4 (Utilities|Struct|MakeMinigameActivityStartedData (Variables|/PapayaGameplay/Minigames/Trials/Waypoints/BP_Waypoint_Parent_Papaya.BP_Waypoint_Parent_Papaya_C|GetMinigameActivityName (Variables|Default|GetStartingWaypoint)) "Time" (Variables|Default|GetTrialTimeRemaining) 0 (Utilities|Array|Length (Variables|Default|GetLinkedWaypoints)))))))
                  (:then_1
                    (Class|BPWaypointParentPapaya|ToggleLinkedVisibility _startingwaypoint true)
                    (Audio|PlaySound2D (Variables|Sound|GetStart%20Sound))
                    (ForceFeedback|SpawnForceFeedbackatLocation (Variables|Default|GetFFE_Start) _returnvalue_3)
                    (Effects|SpawnEmitteratLocation (Variables|Default|GetFX_CheckpointDisappear) _returnvalue_3 (Transformation|GetActorRotation _startingwaypoint) (Variables|Default|GetCheckpointFXScale))
                    (CallFunction|HideTrialStarts true)
                    ;[1 1 1] Can we do this without ticking? Need to revisit
                    (bind _returnvalue_5 (Utilities|Time|SetTimerbyEvent (EventDispatchers|CreateEvent :!event Custom|TimeTrialTimer) (Variables|Default|GetTrialInterval) true))
                    (Variables|Default|SetTrialTimer _returnvalue_5)
                    (Utilities|FlowControl|Sequence
                      (:then_0
                        ;[1 1 1] Checkpoint event
                        (bind (_asynctaskproxy _payload) (Ability|Tasks|WaitGameplayEvent (Variables|Default|GetT_TimeTrialCheckpoint))
                          (:EventReceived
                            (if (Math|Boolean|ANDBoolean (.bSuccess (Utilities|Casting|CastToBP_Waypoint_Parent_Papaya (.OptionalObject (Utilities|Struct|BreakGameplayEventData _payload)))) (Utilities|Array|ContainsItem (Variables|Default|GetLinkedWaypoints) (Utilities|Casting|CastToBP_Waypoint_Parent_Papaya (.OptionalObject (Utilities|Struct|BreakGameplayEventData _payload)))) (not (Utilities|Array|ContainsItem (Variables|Default|GetClearedWaypoints) (Utilities|Casting|CastToBP_Waypoint_Parent_Papaya (.OptionalObject (Utilities|Struct|BreakGameplayEventData _payload))))))
                              (Development|PrintString "CHECKPOINT!")
                              (Variables|Default|SetWaypointActor (Utilities|Casting|CastToBP_Waypoint_Parent_Papaya (.OptionalObject (Utilities|Struct|BreakGameplayEventData _payload))))
                              (Class|BPWaypointParentPapaya|AnimateRingOut (Utilities|Casting|CastToBP_Waypoint_Parent_Papaya (.OptionalObject (Utilities|Struct|BreakGameplayEventData _payload))))
                              (Utilities|Array|AddUnique (Variables|Default|GetClearedWaypoints) (Utilities|Casting|CastToBP_Waypoint_Parent_Papaya (.OptionalObject (Utilities|Struct|BreakGameplayEventData _payload))))
                              ;[1 1 1] Add time when we get a checkpoint
                              (Variables|Default|SetTrialTimeRemaining (+ (Variables|Default|GetTrialTimeRemaining) (.EventMagnitude (Utilities|Struct|BreakGameplayEventData _payload))))
                              (Math|Integer|IncrementInt (Variables|Default|GetCurrentWaypointIdx)
                                (:%20%20
                                  (Utilities|FlowControl|Sequence
                                    (:then_0
                                      (bind _returnvalue_6 (Class|FortControllerComponentMinigameActivity|GetMinigameActivityComponent (Variables|Default|GetInstigatorPawn))
                                        (:Valid
                                          (Class|FortControllerComponentMinigameActivity|ForwardMinigameActivityStatChanged _returnvalue_6 "Time" (Variables|Default|GetTrialTimeRemaining))
                                          (Class|FortControllerComponentMinigameActivity|ForwardMinigameActivityScoreChanged _returnvalue_6 (Variables|Default|GetCurrentWaypointIdx) (Utilities|Array|Length (Variables|Default|GetLinkedWaypoints))))))
                                    (:then_1
                                      (bind _output_get_2 (Variables|Default|SetCheckpointPitchMultiplier (+ (Variables|Default|GetMin%20Pitch) (* (Math|Float|NormalizetoRange (Math|Conversions|ToFloat(Integer) (Variables|Default|GetCurrentWaypointIdx)) 0.0 (Math|Conversions|ToFloat(Integer) (Utilities|Array|Length (Variables|Default|GetLinkedWaypoints)))) (- (Variables|Default|GetMax%20Pitch) (Variables|Default|GetMin%20Pitch)))))) ; Increase pitch up to max on each checkpoint
                                      (Audio|PlaySound2D (Variables|Sound|GetCheckpoint%20Sound) 1.0 _output_get_2)
                                      (ForceFeedback|SpawnForceFeedbackatLocation (Variables|Default|GetFFE_Checkpoint) _returnvalue_2)
                                      (Effects|SpawnEmitteratLocation (Variables|Default|GetFX_CheckpointDisappear) _returnvalue_2 (Transformation|GetActorRotation _waypointactor_2) (Variables|Default|GetCheckpointFXScale)))
                                    (:then_2
                                      (CallFunction|UpdateWaypointColor)))))))))
                      (:then_1
                        ;[1 1 1] Success Event - Goal Checkpoint reached
                        (bind (_asynctaskproxy_2 _payload_2) (Ability|Tasks|WaitGameplayEvent (Variables|Default|GetT_TimeTrialSuccess))
                          (:EventReceived
                            (if (and (.bSuccess (Utilities|Casting|CastToBP_Waypoint_Parent_Papaya (.OptionalObject (Utilities|Struct|BreakGameplayEventData _payload_2)))) (Utilities|Array|ContainsItem (Variables|Default|GetLinkedWaypoints) (Utilities|Casting|CastToBP_Waypoint_Parent_Papaya (.OptionalObject (Utilities|Struct|BreakGameplayEventData _payload_2)))))
                              (Variables|Default|SetWaypointActor (Utilities|Casting|CastToBP_Waypoint_Parent_Papaya (.OptionalObject (Utilities|Struct|BreakGameplayEventData _payload_2))))
                              (Development|PrintString "SUCCESS!")
                              (Utilities|FlowControl|Sequence
                                (:then_0
                                  (Audio|PlaySound2D (Variables|Sound|GetSuccess%20Sound))
                                  (ForceFeedback|SpawnForceFeedbackatLocation (Variables|Default|GetFFE_Finish) _returnvalue)
                                  (Effects|SpawnEmitteratLocation (Variables|Default|GetFX_CheckpointDisappear) _returnvalue (Transformation|GetActorRotation _waypointactor) (Variables|Default|GetCheckpointFXScale)))
                                (:then_1
                                  (Class|BPWaypointParentPapaya|ToggleLinkedVisibility (Variables|Default|GetStartingWaypoint) false)
                                  (Math|Integer|IncrementInt (Variables|Default|GetCurrentWaypointIdx)
                                    (:%20%20
                                      (Utilities|Time|ClearandInvalidateTimerbyHandle (Variables|Default|GetTrialTimer))
                                      (CallFunction|UpdateTopScoreBoard))))))))))))))))))
    (:then_2
      (CallFunction|UpdateWaypointColor))))

(event Custom|TrialFailed
  ;[1 1 1] Trial Failed
  (if (Utilities|Time|IsValidTimerHandle (Variables|Default|GetTrialTimer))
    (Development|PrintString "TRIAL FAILED" true true "(R=1.000000,G=0.041157,B=0.071004,A=1.000000)")
    (Audio|PlaySound2D (Variables|Sound|GetFailed%20Sound))
    (Class|BPWaypointParentPapaya|ToggleLinkedVisibility (Variables|Default|GetStartingWaypoint) false)
    (Utilities|Time|ClearandInvalidateTimerbyHandle (Variables|Default|GetTrialTimer))
    (bind _returnvalue (Class|FortControllerComponentMinigameActivity|GetMinigameActivityComponent (Variables|Default|GetInstigatorPawn))
      (:Valid
        (Class|FortControllerComponentMinigameActivity|ForwardMinigameActivityEnded _returnvalue (Utilities|Struct|MakeMinigameActivityEndedData false "Time" 0.0 (CallFunction|GetTopScore) (Variables|Default|GetCurrentWaypointIdx) (Utilities|Array|Length (Variables|Default|GetLinkedWaypoints)))))
      (:IsNotValid
        _))
    (Ability|EndAbility)))

(event Ability|EventOnEndAbility
  (params (bWasCancelled :type Boolean))
  ;[1 1 1] End Ability -> remove GE and unhide TrialStarts
  (CallFunction|HideTrialStarts false))

(event Custom|UpdateTopScoreBoard
  ;[1 1 1] Update the minigame ui and top score board on success
  (bind _output_get (Variables|Default|SetEndTime (Fortnite|Sync|Time|GetServerWorldTimeSeconds)))
  (bind _output_get_2 (Variables|Default|SetFinalScore (+ (- _output_get (Variables|Default|GetStartTime)) (* (- (Utilities|Array|Length (Variables|Default|GetLinkedWaypoints)) (Variables|Default|GetCurrentWaypointIdx)) (Variables|Default|GetSkippedWaypointTimePenalty)))))
  (Development|PrintString (Utilities|String|BuildString(Float) "" "FINAL TIME: " _output_get_2) true true "(R=1.000000,G=0.167282,B=0.000000,A=1.000000)" 1e+01)
  (CallFunction|UpdateScoreboard (Variables|Default|GetFinalScore))
  (bind _returnvalue (Class|FortControllerComponentMinigameActivity|GetMinigameActivityComponent (Variables|Default|GetInstigatorPawn))
    (:Valid
      (Class|FortControllerComponentMinigameActivity|ForwardMinigameActivityEnded _returnvalue (Utilities|Struct|MakeMinigameActivityEndedData true "Time" (Variables|Default|GetFinalScore) (CallFunction|GetTopScore) (Variables|Default|GetCurrentWaypointIdx) (Utilities|Array|Length (Variables|Default|GetLinkedWaypoints)) (CallFunction|ScoreRank (Variables|Default|GetFinalScore)))))
    (:IsNotValid
      _))
  (Ability|Tasks|WaitNetSync "OnlyServerWait"
    (:OnSync
      (Ability|EndAbility))))

(event Custom|UpdateScoreboard
  (params (NewTime :type Float%20%28double-precision%29))
  (bind _startingwaypoint (Variables|Default|GetStartingWaypoint))
  (bind _externalscoreboardactor (Variables|/PapayaGameplay/Minigames/Trials/Waypoints/BP_Waypoint_Parent_Papaya.BP_Waypoint_Parent_Papaya_C|GetExternalScoreboardActor _startingwaypoint))
  ;[1 1 1] Update the minigame ui and top score board on success
  (Utilities|IsValid _externalscoreboardactor
    (:Is%20Valid
      (Class|BPapayaTrialsTopScore|UpdateTopScore _externalscoreboardactor (Utilities|Struct|MakeStructPapayaTrialsTopScore NewTime (Utilities|Text|ToText(String) (Fortnite|GetPlayerNameSafe (Variables|Default|GetInstigatorPawn))))))))

(event Custom|Reset%20Time%20Trial%20Warning%20Audio
  ;[1 1 1] Warning Audio WIP
  (Utilities|FlowControl|DoOnce false
    (:Completed
      (Development|PrintString "5 seconds left - player 
warning audio or change 
pitch or w/e the fuck"))))

(event Custom|TimeTrialTimer
  (bind _output_get (Variables|Default|SetTrialTimeRemaining (- (Variables|Default|GetTrialTimeRemaining) (Variables|Default|GetTrialInterval))))
  (Utilities|FlowControl|Sequence
    (:then_0
      (bind _returnvalue (Class|FortControllerComponentMinigameActivity|GetMinigameActivityComponent (Variables|Default|GetInstigatorPawn))
        (:Valid
          (Class|FortControllerComponentMinigameActivity|ForwardMinigameActivityStatChanged _returnvalue "Time" (Variables|Default|GetTrialTimeRemaining)))))
    (:then_1
      (Development|PrintString (Utilities|String|Append "Trial Time Remaining: " (Utilities|String|ToString(Float) _output_get)) (Variables|Default|GetDebug) (Variables|Default|GetDebug) "(R=1.000000,G=0.455128,B=0.054146,A=1.000000)")
      (if (<= _output_get 0.0)
        (CallFunction|TrialFailed)))))

(fn GetTopScore
  (Utilities|FlowControl|Sequence
    (:then_0
      (bind _startingwaypoint (Variables|Default|GetStartingWaypoint)
        (:then
          (Utilities|IsValid (Variables|/PapayaGameplay/Minigames/Trials/Waypoints/BP_Waypoint_Parent_Papaya.BP_Waypoint_Parent_Papaya_C|GetExternalScoreboardActor _startingwaypoint)
            (:Is%20Valid
              (LocalVariables|SetLocal_TopScore (Utilities|Struct|BreakStructPapayaTrialsTopScore (Variables|/PapayaGameplay/Minigames/Trials/B_Papaya_Trials_TopScore.B_Papaya_Trials_TopScore_C|GetStruct_TopScore (Variables|/PapayaGameplay/Minigames/Trials/Waypoints/BP_Waypoint_Parent_Papaya.BP_Waypoint_Parent_Papaya_C|GetExternalScoreboardActor _startingwaypoint)))))))
        (:else)))
    (:then_1
      (return (LocalVariables|GetLocal_TopScore)))))

(fn HideTrialStarts
  (params (Hide :type Boolean))
  (bind _allwaypoints (Variables|Default|GetAllWaypoints))
  (if Hide
    (bind _outactors (Actor|GetAllActorsOfClass "/PapayaGameplay/Minigames/Trials/Waypoints/BP_Waypoint_Parent_Papaya.BP_Waypoint_Parent_Papaya_C"))
    (if (Utilities|IsArrayNonEmpty _outactors)
      (bind _output_get (Variables|Default|SetAllWaypoints _outactors))
      (for _array_element _output_get
        (Utilities|IsValid _array_element
          (:Is%20Valid
            (if (and (Variables|/PapayaGameplay/Minigames/Trials/Waypoints/BP_Waypoint_Parent_Papaya.BP_Waypoint_Parent_Papaya_C|GetStartsTrial _array_element) (Utilities|NotEqual(Object) _array_element self))
              (if (not (Variables|/Script/Engine.Actor|GetbHidden _array_element))
                (Class|BPWaypointParentPapaya|AnimateRingOut _array_element)))))))
    (elif (Utilities|IsArrayNonEmpty _allwaypoints)
      (for _array_element_2 _allwaypoints
        (Utilities|IsValid _array_element_2
          (:Is%20Valid
            (if (Math|Boolean|ANDBoolean (Variables|/PapayaGameplay/Minigames/Trials/Waypoints/BP_Waypoint_Parent_Papaya.BP_Waypoint_Parent_Papaya_C|GetStartsTrial _array_element_2) (Utilities|NotEqual(Object) _array_element_2 self) (not (.bSuccess (Utilities|Casting|CastToBP_Waypoint_Papaya_Skydive_Start _array_element_2))))
              (Class|BPWaypointParentPapaya|AnimateRingIn _array_element_2))))))))

(fn ScoreRank
  (params (FinalScore :type Float%20%28double-precision%29))
  (bind _startingwaypoint (Variables|Default|GetStartingWaypoint))
  (Utilities|IsValid _startingwaypoint
    (:Is%20Valid
      (if (<= FinalScore (Variables|/PapayaGameplay/Minigames/Trials/Waypoints/BP_Waypoint_Parent_Papaya.BP_Waypoint_Parent_Papaya_C|GetParTime_Rank0 _startingwaypoint))
        (return 0)
        (elif (<= FinalScore (Variables|/PapayaGameplay/Minigames/Trials/Waypoints/BP_Waypoint_Parent_Papaya.BP_Waypoint_Parent_Papaya_C|GetParTime_Rank1 _startingwaypoint))
          (return 1)
          (elif (<= FinalScore (Variables|/PapayaGameplay/Minigames/Trials/Waypoints/BP_Waypoint_Parent_Papaya.BP_Waypoint_Parent_Papaya_C|GetParTime_Rank2 _startingwaypoint))
            (return 2)
            (elif (<= (Variables|/PapayaGameplay/Minigames/Trials/Waypoints/BP_Waypoint_Parent_Papaya.BP_Waypoint_Parent_Papaya_C|GetParTime_Rank3 _startingwaypoint) FinalScore)
              (return 3))))))
    (:Is%20Not%20Valid
      _))
  (return 4))

(fn UpdateWaypointColor
  (LocalVariables|SetWaypointsToUpdate (Variables|Default|GetLinkedWaypoints))
  (for _array_element (Variables|Default|GetClearedWaypoints)
    (Utilities|Array|RemoveItem (LocalVariables|GetWaypointsToUpdate) _array_element))
  (for (_array_element_2 _array_index) (LocalVariables|GetWaypointsToUpdate)
    (Utilities|IsValid _array_element_2
      (:Is%20Valid
        (Class|BPWaypointParentPapaya|UpdateMaterialsBySequence _array_element_2 _array_index)))))
